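# Physics asset for the dangling holster and gun on the Weird Objects Police medium male player body.
# Four bodies form a small rig: pelvis anchors the holster, the holster carries the gun, and a right-thigh capsule provides a collision volume.
# The holster swings within limits from the pelvis, while the gun is nearly rigid to it with swing2 and twist locked.
# The thigh-to-pelvis joint leaves all angular axes free.
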PhysicsAsset /BRCosmetics/Characters/Player/Male/Medium/Bodies/M_MED_Weird_Objects_Police/Meshes/M_MED_Weird_Objects_Police_Dyn
bodies {
 dyn_gun [{shapeType=Capsule, length=3, radius=2, center=<-5 0 0>, rotation=<rot 90 0 0>}]
 dyn_holster [
  {shapeType=Capsule, length=14.1569, radius=1.9885, center=<5 0 0>, rotation=<rot 90 0 0>}
  {shapeType=Box, extentX=12.1871, extentY=7.5, extentZ=5.2671, center=<20.9261 -0.0001 0>, rotation=<rot -10 0 0>}
 ]
 pelvis [{shapeType=Capsule, length=3, radius=2, rotation=<rot 90 0 0>}]
 thigh_r [{shapeType=Capsule, length=22.6999, radius=8.5523, center=<19.704 0.0005 0>, rotation=<rot 90 0 0>}]
}
constraints [
 {
  bone1Name=dyn_gun
  bone2Name=dyn_holster
  swing1LimitDegrees=5
  swing2LimitDegrees=45
  swing2Motion=Locked
  twistLimitDegrees=5
  twistMotion=Locked
 }
 {bone1Name=dyn_holster, bone2Name=pelvis, swing1LimitDegrees=45, swing2LimitDegrees=60, twistLimitDegrees=5}
 {
  bone1Name=thigh_r
  bone2Name=pelvis
  swing1LimitDegrees=45
  swing1Motion=Free
  swing2LimitDegrees=45
  swing2Motion=Free
  twistLimitDegrees=45
  twistMotion=Free
 }
]
